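# Landscape material for the Papaya Mountains biome, defining terrain shading, grass spawning and physical surfaces.
# Common layers height-blend with painted scree, forest floor, heather and autumn leaves, then gain slope cliffs and procedural snow.
# The result writes to and samples a runtime virtual texture, with a simpler mobile path and a final Curie pass.
# Layer, cliff and snow masks drive baked grass output, whose inputs match every biome, and physical materials.
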Material /Game/Athena/Apollo/Environments/Landscape/Materials/Biome/LandscapeBiomeMaterials/Papaya/M_Biome_Papaya_Mountains
graph {
 properties {
  bTangentSpaceNormal=false
  bUseMaterialAttributes=true
  bUsedWithVirtualHeightfieldMesh=true
  pixelDepthOffsetMode=PDOM_Legacy
  usedShadingModels=Default Lit
 }
 nodes [
  n1 = Saturate(None=Multiply(A=Constant() {r=2}, B=LandscapeLayerSample() {parameterName=BiomeD}))
  n2 = LandscapeVisibilityMask()
  n3 = Max(A=NamedRerouteUsage(), B=LandscapeLayerSample() {parameterName=Mud})
  n4 = LandscapeLayerSample() {parameterName=Sand}
  n5 = Saturate(None=Multiply(A=Constant() {r=2}, B=@n4))
  n6 = Ceil(None=LandscapeLayerSample() {parameterName=Asphalt})
  n7 = Max(A=Max(A=@n3, B=@n5), B=@n6)
  n8 = Saturate(None=LinearInterpolate(
    A=ScalarParameter() {defaultValue=-0.5, parameterName=BiomeA-GrassSpawnLa}
    B=ScalarParameter() {defaultValue=3, parameterName=BiomeA-GrassSpawnLb}
    Alpha=LandscapeLayerSample() {desc=Scree, parameterName=BiomeA}
   ))
  n9 = NamedRerouteUsage()
  n10 = ScalarParameter() {defaultValue=5, parameterName=BiomeA-GrassSpawnCont}
  n11 = ScalarParameter() {defaultValue=-0.5, parameterName=BiomeB-FFSpawnLa}
  n12 = ScalarParameter() {defaultValue=1.5, parameterName=BiomeB-FFSpawnLb}
  n13 = LandscapeLayerSample() {desc=ForestFloor, parameterName=BiomeB}
  n14 = MaterialFunctionCall(
   Transition Phase=Saturate(None=LinearInterpolate(A=@n11, B=@n12, Alpha=@n13))
   Height Texture=NamedRerouteUsage()
   Contrast=ScalarParameter() {defaultValue=0.5, parameterName=BiomeB-FFSpawnCont}
  ) {desc=ForestFloor Masked, materialFunction=&HeightLerp}
  n15 = LandscapeLayerSample() {desc=Heather, parameterName=BiomeC}
  n16 = Max(
   A=Max(
    A=MaterialFunctionCall(Transition Phase=@n8, Height Texture=@n9, Contrast=@n10) {desc=Scree Masked, materialFunction=&HeightLerp}.Alpha
    B=@n14.Alpha
   )
   B=@n15
  )
  n17 = LandscapeLayerSample() {desc=Scree Layer, parameterName=BiomeA}
  n18 = NamedRerouteUsage()
  n19 = Max(A=@n18, B=Max(A=Max(A=@n7, B=@n16), B=@n17))
  n20 = NamedRerouteUsage()
  n21 = Multiply(A=@n20, B=Constant() {r=2})
  n22 = Multiply(A=@n2, B=Subtract(A=OneMinus(None=@n19), B=Saturate(None=@n21)))
  MaterialExpressionReroute_4 = Reroute(None=Multiply(A=OneMinus(None=@n1), B=@n22))
  MaterialExpressionConstant_8 = Constant()
  n23 = LinearInterpolate(
   A=ScalarParameter() {defaultValue=-4, parameterName=Threshold_Scree}
   B=ScalarParameter() {defaultValue=1.6832, parameterName=Threshold_Scree2}
   Alpha=ComponentMask(None=VertexNormalWS()) {b=true}
  )
  n24 = Constant() {r=2}
  n25 = Constant() {r=-0.5}
  n26 = Add(A=Multiply(A=@n24, B=@n17), B=@n25)
  MaterialExpressionMultiply_9 = Multiply(A=@n2, B=Multiply(A=Saturate(None=@n23), B=Saturate(None=@n26)))
  n27 = Constant() {r=3}
  n28 = Subtract(A=Saturate(None=Multiply(A=@n27, B=@n17)), B=@n18)
  n29 = Max(A=Max(A=@n3, B=@n5), B=@n6)
  n30 = Multiply(A=@n20, B=Constant() {r=20})
  MaterialExpressionMultiply_7 = Multiply(A=@n2, B=Subtract(A=Subtract(A=@n28, B=@n29), B=Saturate(None=@n30)))
  n31 = Add(A=@n25, B=Multiply(A=@n24, B=@n15))
  MaterialExpressionReroute_25 = Reroute(None=Multiply(A=Saturate(None=@n31), B=OneMinus(None=@n18)))
  MaterialExpressionReroute_53 = Reroute(None=@n14.Alpha)
  MaterialExpressionReroute_20 = Reroute(None=Subtract(A=@n4, B=@n18))
  MaterialExpressionMultiply_16 = Multiply(A=@n1, B=@n22)
  LandscapeGrassOutput!(
   Grass_Standard=@MaterialExpressionReroute_4
   Bush=@MaterialExpressionReroute_4
   FarmFlowers=@MaterialExpressionConstant_8
   Scree=@MaterialExpressionMultiply_9
   Scree_Grass=@MaterialExpressionMultiply_7
   Heather=@MaterialExpressionReroute_25
   SwampGrass=@MaterialExpressionConstant_8
   SwampReeds=@MaterialExpressionConstant_8
   Forest_Grass=@MaterialExpressionReroute_53
   Beach=@MaterialExpressionReroute_20
   Ocean=@MaterialExpressionConstant_8
   Nightmare-Grass=@MaterialExpressionConstant_8
   Nightmare-Leaves=@MaterialExpressionConstant_8
   AutumnLeafArea=@MaterialExpressionMultiply_16
  ) {
   desc="This must have the exact inputs in every biome material!\r\nThis data is baked into the level data when the level is\r\nsaved, so any changes to this require the affected level\r\nto be saved && submitted."
   grassTypes [
    Grass_Standard {
     grassType=/Game/Athena/Apollo/Environments/Landscape/SmallFoliage/DataAssets/LandscapeGrassType/Papaya/LGT_Papaya_StandardGrass
     input {
      expression=&self:MaterialExpressionReroute_4
      inputName=None
      mask=0
      maskA=0
      maskB=0
      maskG=0
      maskR=0
      outputIndex=0
     }
    }
    Bush {
     grassType=/Game/Athena/Environments/Landscape/Foliage/Athena_Foliage_Bushes_Papaya
     input {
      expression=&self:MaterialExpressionReroute_4
      inputName=None
      mask=0
      maskA=0
      maskB=0
      maskG=0
      maskR=0
      outputIndex=0
     }
    }
    FarmFlowers {
     grassType=/Game/Athena/Apollo/Environments/Landscape/SmallFoliage/DataAssets/LandscapeGrassType/LandscapeGrassType_FarmFlowers
     input {
      expression=&self:MaterialExpressionConstant_8
      inputName=None
      mask=0
      maskA=0
      maskB=0
      maskG=0
      maskR=0
      outputIndex=0
     }
    }
    Scree {
     grassType=/Game/Athena/Apollo/Environments/Landscape/SmallFoliage/DataAssets/LandscapeGrassType/LandscapeGrassType_Scree
     input {
      expression=&self:MaterialExpressionMultiply_9
      inputName=None
      mask=0
      maskA=0
      maskB=0
      maskG=0
      maskR=0
      outputIndex=0
     }
    }
    Scree_Grass {
     grassType=/Game/Athena/Apollo/Environments/Landscape/SmallFoliage/DataAssets/LandscapeGrassType/Papaya/LGT_Papaya_ScreeGrass
     input {
      expression=&self:MaterialExpressionMultiply_7
      inputName=None
      mask=0
      maskA=0
      maskB=0
      maskG=0
      maskR=0
      outputIndex=0
     }
    }
    Heather {
     grassType=/Game/Athena/Apollo/Environments/Landscape/SmallFoliage/DataAssets/LandscapeGrassType/LandscapeGrassType_Heather
     input {
      expression=&self:MaterialExpressionReroute_25
      inputName=None
      mask=0
      maskA=0
      maskB=0
      maskG=0
      maskR=0
      outputIndex=0
     }
    }
    SwampGrass {
     grassType=/Game/Athena/Apollo/Environments/Landscape/SmallFoliage/DataAssets/LandscapeGrassType/LandscapeGrassType_SwampGrass
     input {
      expression=&self:MaterialExpressionConstant_8
      inputName=None
      mask=0
      maskA=0
      maskB=0
      maskG=0
      maskR=0
      outputIndex=0
     }
    }
    SwampReeds {
     grassType=/Game/Athena/Apollo/Environments/Landscape/SmallFoliage/DataAssets/LandscapeGrassType/LandscapeGrassType_SwampReeds
     input {
      expression=&self:MaterialExpressionConstant_8
      inputName=None
      mask=0
      maskA=0
      maskB=0
      maskG=0
      maskR=0
      outputIndex=0
     }
    }
    Forest_Grass {
     grassType=/Game/Athena/Apollo/Environments/Landscape/SmallFoliage/DataAssets/LandscapeGrassType/LandscapeGrassType_ForestGrass
     input {
      expression=&self:MaterialExpressionReroute_53
      inputName=None
      mask=0
      maskA=0
      maskB=0
      maskG=0
      maskR=0
      outputIndex=0
     }
    }
    Beach {
     grassType=/Game/Athena/Apollo/Environments/Landscape/SmallFoliage/DataAssets/LandscapeGrassType/LandscapeGrassType_Beach
     input {
      expression=&self:MaterialExpressionReroute_20
      inputName=None
      mask=0
      maskA=0
      maskB=0
      maskG=0
      maskR=0
      outputIndex=0
     }
    }
    Ocean {
     grassType=/Game/Athena/Apollo/Environments/Landscape/SmallFoliage/DataAssets/LandscapeGrassType/LandscapeGrassType_Ocean
     input {
      expression=&self:MaterialExpressionConstant_8
      inputName=None
      mask=0
      maskA=0
      maskB=0
      maskG=0
      maskR=0
      outputIndex=0
     }
    }
    Nightmare-Grass {
     grassType=/Game/Athena/Apollo/Environments/Landscape/SmallFoliage/DataAssets/LandscapeGrassType/LandscapeGrassType_Nightmare
     input {
      expression=&self:MaterialExpressionConstant_8
      inputName=None
      mask=0
      maskA=0
      maskB=0
      maskG=0
      maskR=0
      outputIndex=0
     }
    }
    Nightmare-Leaves {
     grassType=/Game/Athena/Apollo/Environments/Landscape/SmallFoliage/DataAssets/LandscapeGrassType/LandscapeGrassType_Nightmare-Leaves
     input {
      expression=&self:MaterialExpressionConstant_8
      inputName=None
      mask=0
      maskA=0
      maskB=0
      maskG=0
      maskR=0
      outputIndex=0
     }
    }
    AutumnLeafArea {
     grassType=/Game/Athena/Apollo/Environments/Landscape/SmallFoliage/DataAssets/LandscapeGrassType/LandscapeGrassType_AutumnLeafArea
     input {
      expression=&self:MaterialExpressionMultiply_16
      inputName=None
      mask=0
      maskA=0
      maskB=0
      maskG=0
      maskR=0
      outputIndex=0
     }
    }
   ]
  }
  n32 = NamedRerouteUsage()
  n33 = OneMinus(None=@n32) {desc="White=Cliffs"}
  n34 = Max(A=LandscapeLayerSample() {parameterName=Gravel}, B=LandscapeLayerSample() {parameterName=Mud})
  n35 = Constant() {r=2}
  n36 = Max(A=Ceil(None=@n20), B=Saturate(None=Multiply(A=@n34, B=@n35))) {desc=Add Snow}
  n37 = Subtract(A=@n36, B=@n33) {desc="Remove Cliff from Gravel,Mud, Sand, Snow"}
  n38 = LandscapeLayerSample() {parameterName=Sand}
  n39 = Saturate(None=Multiply(A=LandscapeLayerSample() {parameterName=Asphalt}, B=Constant() {r=3}))
  n40 = Max(A=Max(A=@n14.Alpha, B=@n15), B=@n39)
  MaterialExpressionSaturate_12 = Saturate(None=Subtract(A=Saturate(None=@n37), B=Max(A=@n38, B=@n40)) {desc="Remove ForestFloor, Heather, Asphalt, Sand"}) {desc=Final Dirt mask}
  MaterialExpressionReroute_51 = Reroute(None=@n33) {desc=Final Stone mask}
  n41 = Subtract(A=@n32, B=Max(A=@n38, B=@n36)) {desc="Remove Gravel,Mud, Sand, Snow"}
  MaterialExpressionSaturate_14 = Saturate(None=Subtract(A=Saturate(None=@n41), B=@n39) {desc=Remove Asphalt}) {desc=Final Grass mask}
  MaterialExpressionReroute_61 = Reroute(None=@n39)
  n42 = Saturate(None=Subtract(A=@n38, B=@n33) {desc=Remove Cliff})
  MaterialExpressionReroute_7 = Reroute(None=Saturate(None=Subtract(A=@n42, B=@n40) {desc="Remove ForestFloor, Flowers, Asphalt"}))
  LandscapePhysicalMaterialOutput!(
   Dirt=@MaterialExpressionSaturate_12
   Stone=@MaterialExpressionReroute_51
   Grass=@MaterialExpressionSaturate_14
   Asphalt=@MaterialExpressionReroute_61
   Sand=@MaterialExpressionReroute_7
  ) {
   desc="These are layered in order of importance.\r\nIf Dirt=1 and Stone=1, then Dirt will \"win\".\r\nTheoretically, these values aren't clamped.\r\nSo a value of 2 should trump a value of 1."
   inputs [
    {
     physicalMaterial=/Game/FortressPhysicalMaterials/Dirt
     input {
      expression=&self:MaterialExpressionSaturate_12
      inputName=None
      mask=0
      maskA=0
      maskB=0
      maskG=0
      maskR=0
      outputIndex=0
     }
    }
    {
     physicalMaterial=/Game/FortressPhysicalMaterials/Stone
     input {
      expression=&self:MaterialExpressionReroute_51
      inputName=None
      mask=0
      maskA=0
      maskB=0
      maskG=0
      maskR=0
      outputIndex=0
     }
    }
    {
     physicalMaterial=/Game/FortressPhysicalMaterials/Grass
     input {
      expression=&self:MaterialExpressionSaturate_14
      inputName=None
      mask=0
      maskA=0
      maskB=0
      maskG=0
      maskR=0
      outputIndex=0
     }
    }
    {
     physicalMaterial=/Game/FortressPhysicalMaterials/Asphalt
     input {
      expression=&self:MaterialExpressionReroute_61
      inputName=None
      mask=0
      maskA=0
      maskB=0
      maskG=0
      maskR=0
      outputIndex=0
     }
    }
    {
     physicalMaterial=/Game/FortressPhysicalMaterials/Sand
     input {
      expression=&self:MaterialExpressionReroute_7
      inputName=None
      mask=0
      maskA=0
      maskB=0
      maskG=0
      maskR=0
      outputIndex=0
     }
    }
   ]
  }
  n43 = LinearInterpolate(A=@MaterialExpressionSaturate_12, B=@MaterialExpressionReroute_51, Alpha=ScalarParameter() {parameterName=DEBUG-PhysMat1})
  n44 = ScalarParameter() {parameterName=DEBUG-PhysMat2}
  LinearInterpolate(
   A=LinearInterpolate(
    A=LinearInterpolate(A=@n43, B=@MaterialExpressionSaturate_14, Alpha=@n44)
    B=@MaterialExpressionReroute_61
    Alpha=ScalarParameter() {parameterName=DEBUG-PhysMat3}
   )
   B=@MaterialExpressionReroute_7
   Alpha=ScalarParameter() {defaultValue=1, parameterName=DEBUG-PhysMat4}
  )
  n45 = GetMaterialAttributes(None=MaterialFunctionCall!(PutSandBeforeGravelOverride, UseSnow=NamedRerouteUsage(), UseOceanLevel) {materialFunction=/Game/Athena/Apollo/Environments/Landscape/Materials/Biome/CommonMaterials/Mobile/MF_Mobile_CommonLandscapeLayers}.Material) {attributeGetTypes [#1]}
  MaterialExpressionConstant_13 = Constant()
  n46 = LandscapeLayerBlend!(
   Layer Gravel=@MaterialExpressionConstant_13
   Layer Mud=@MaterialExpressionConstant_13
   Layer Sand=@MaterialExpressionConstant_13
   Layer Asphalt=@MaterialExpressionConstant_13
  ) {layers [
    {
     blendType=LB_WeightBlend
     constHeightInput=0
     layerName=Gravel
     previewWeight=0
     constLayerInput=<0 0 0>
     heightInput {expression=null, inputName=None, mask=0, maskA=0, maskB=0, maskG=0, maskR=0, outputIndex=0}
     layerInput {
      expression=&self:MaterialExpressionConstant_13
      inputName=None
      mask=0
      maskA=0
      maskB=0
      maskG=0
      maskR=0
      outputIndex=0
     }
    }
    {
     blendType=LB_WeightBlend
     constHeightInput=0
     layerName=Mud
     previewWeight=0
     constLayerInput=<0 0 0>
     heightInput {expression=null, inputName=None, mask=0, maskA=0, maskB=0, maskG=0, maskR=0, outputIndex=0}
     layerInput {
      expression=&self:MaterialExpressionConstant_13
      inputName=None
      mask=0
      maskA=0
      maskB=0
      maskG=0
      maskR=0
      outputIndex=0
     }
    }
    {
     blendType=LB_WeightBlend
     constHeightInput=0
     layerName=Sand
     previewWeight=0
     constLayerInput=<0 0 0>
     heightInput {expression=null, inputName=None, mask=0, maskA=0, maskB=0, maskG=0, maskR=0, outputIndex=0}
     layerInput {
      expression=&self:MaterialExpressionConstant_13
      inputName=None
      mask=0
      maskA=0
      maskB=0
      maskG=0
      maskR=0
      outputIndex=0
     }
    }
    {
     blendType=LB_WeightBlend
     constHeightInput=0
     layerName=Asphalt
     previewWeight=0
     constLayerInput=<0 0 0>
     heightInput {expression=null, inputName=None, mask=0, maskA=0, maskB=0, maskG=0, maskR=0, outputIndex=0}
     layerInput {
      expression=&self:MaterialExpressionConstant_13
      inputName=None
      mask=0
      maskA=0
      maskB=0
      maskG=0
      maskR=0
      outputIndex=0
     }
    }
   ]}
  n47 = SetMaterialAttributes!(MaterialAttributes=@n45.MaterialAttributes, Base Color=Add(A=@n45.Base Color, B=Multiply(A=@n46) {constB=0})) {desc="This must remain in material,\r\nnot a function.", attributeSetTypes [#1]}
  n48 = RuntimeVirtualTextureSampleParameter() {
   group=RVT
   materialType=BaseColor_Normal_Specular_YCoCg
   parameterName=RVT Sample Target
   virtualTexture=/Game/Athena/Apollo/Maps/Landscape/VT/RVT_Apollo_Landscape
  }
  n49 = SetMaterialAttributes!(MaterialAttributes, Base Color=@n48.BaseColor, Specular=@n48.BaseColor, Roughness=@n48.BaseColor, Normal=@n48.BaseColor) {attributeSetTypes [
    #1
    #2
    #3
    #4
   ]}
  n50 = MaterialFunctionCall!(AdditionalGrassBorderColorMask=LandscapeLayerSample() {desc=ForestFloor, parameterName=BiomeB}, PutSandBeforeGravel, UseOceanLevel) {materialFunction=/Game/Athena/Apollo/Environments/Landscape/Materials/Biome/CommonMaterials/MF_CommonLandscapeLayers}
  n51 = LandscapeLayerSample() {desc=Scree, parameterName=BiomeA}
  n52 = MaterialFunctionCall!(LayerWeight=@n51) {materialFunction=/Game/Athena/Apollo/Environments/Landscape/Materials/Biome/UniqueBiomeMaterialFunctions/MF_Scree}
  n53 = MaterialFunctionCall!(
   Material 1=@n50.Material
   Height Texture 1=@n50.Material
   Material 2=@n52.Material
   Height Texture 2=NamedRerouteDeclaration(None=@n52.Height) {name=MF_Scree-Height, nodeColor=<rgba 0.5 0 1 1>}
   Transistion Phase=@n51
   UseContrast=StaticBool() {value=true}
   Contrast=Constant() {r=0.5}
   AddLayerMask
  ) {materialFunction=/Game/Athena/Apollo/Environments/Landscape/Materials/Biome/CommonMaterials/MF_BlendLandscapeLayers}
  n54 = MaterialFunctionCall() {materialFunction=/Game/Athena/Apollo/Environments/Landscape/Materials/Biome/UniqueBiomeMaterialFunctions/MF_Biome_Forest}
  n55 = MaterialFunctionCall(
   Material 1=@n53.BlendedMaterial
   Height Texture 1=@n53.BlendedMaterial
   Material 2=@n54.Material
   Height Texture 2=NamedRerouteDeclaration(None=@n54.Height) {name=MF_Biome_Forest-Height, nodeColor=<rgba 0.5 0 1 1>}
   Transistion Phase=LandscapeLayerSample() {desc=ForestFloor, parameterName=BiomeB}
  ) {materialFunction=&MF_BlendLandscapeLayers}
  n56 = MaterialFunctionCall!(GrassHeightmap=NamedRerouteUsage(), GrassNoiseMask=NamedRerouteUsage(), GrassMaterial=NamedRerouteUsage()) {materialFunction=/Game/Athena/Apollo/Environments/Landscape/Materials/Biome/UniqueBiomeMaterialFunctions/MF_Biome_Heather}
  n57 = MaterialFunctionCall(
   Material 1=@n55.BlendedMaterial
   Height Texture 1=@n55.BlendedMaterial
   Material 2=@n56.Material
   Height Texture 2=@n56.Material
   Transistion Phase=LandscapeLayerSample() {parameterName=BiomeC}
  ) {materialFunction=&MF_BlendLandscapeLayers}
  n58 = MaterialFunctionCall!(Material=@n57.BlendedMaterial, LandscapeLayer=LandscapeLayerSample() {parameterName=BiomeD}) {materialFunction=/Game/Athena/Apollo/Environments/Landscape/Materials/Biome/UniqueBiomeMaterialFunctions/MF_Biome_Autumn_Leaves}
  n59 = MaterialFunctionCall(
   Material 1=@n57.BlendedMaterial
   Height Texture 1=@n57.BlendedMaterial
   Material 2=@n58.Material
   Height Texture 2=@n58.Material
   Transistion Phase=@n58.Material
   UseContrast=StaticBool() {value=true}
   Contrast=@n58.Material
  ) {materialFunction=&MF_BlendLandscapeLayers}
  n60 = MaterialFunctionCall!(
   Material=@n59.BlendedMaterial
   Heightmap=@n59.BlendedMaterial
   NewSlopeMaskMask
   Displacement
   AsphaltLayer=LandscapeLayerSample() {parameterName=Sand}
   UseAlternateCliffMat
   AltCliffHeight
   AltCliffMat
  ) {materialFunction=/Game/Athena/Apollo/Environments/Landscape/Materials/Biome/CommonMaterials/MF_SlopeBlendCliff}
  n61 = MaterialFunctionCall!(Material=@n60.Material) {materialFunction=/Game/Athena/Apollo/Maps/Landscape/VT/MaterialFunctions/TransformTangentToWorld}
  n62 = MaterialFunctionCall!(Heightmap=@n60.RegularHeightmap, WorldnormalIn=@n61.Normals, UseManualSnowMask, SnowMask) {materialFunction=/Game/Athena/Apollo/Environments/Landscape/Materials/Biome/UniqueBiomeMaterialFunctions/MF_ProceduralSnow}
  n63 = StaticBool() {value=true}
  n64 = MaterialFunctionCall!(
   MobileMaterial=@n47
   HighEndMaterial=StaticSwitch(
    True=BlendMaterialAttributes(A=@n61.Result, B=@n62.Material, Alpha=@n62.Material)
    False=@n61.Result
    Value=NamedRerouteDeclaration(None=@n63) {name=UseSnow, nodeColor=<rgba 0.5 0 1 1>}
   ) {desc=Use Procedural Snow}
  ) {materialFunction=/Game/Athena/Asteria/Environments/Landscape/Materials/MaterialFunctions/FN_RvtWriteSwitch}
  n65 = MaterialFunctionCall(
   NoRVT=@n47
   RVT=StaticSwitchParameter(
    True=MaterialFunctionCall!(RVT=@n49, ByPass RVT=@n64.Result) {materialFunction=/Game/Athena/Apollo/Maps/Landscape/VT/MaterialFunctions/EnableRVT}.Result
    False=@n64.Result
   ) {defaultValue=true, group=RVT Base, parameterName=RVT Enable}
  ) {materialFunction=&FN_RVT_Switch_Material}
  n66 = SetMaterialAttributes!(MaterialAttributes=@n65.Result, Opacity Mask=LandscapeVisibilityMask()) {attributeSetTypes [#5]}
  MaterialFunctionCall!(
   In=ShadingPathSwitch(Default=@n66, Deferred=@n66, Forward=@n66, Mobile=@n65.Result)
   UseHeightFalloff
   OnFireEmissiveOverride
   HeightFalloffRadius
   HeightFalloffHardness
   VariationMasksMobile
   In_Grass
   Is_Stylised_WE
  ) {desc=This should be very last, materialFunction=/Game/Packages/Fortress_SharedMaterials/Base_Material_Functions/MF_Curie_Landscape}
  NamedRerouteDeclaration(None=@n50.GrassNoiseMask) {name=GrassNoiseMask, nodeColor=<rgba 0.5 0 1 1>}
  NamedRerouteDeclaration(None=@n50.Material) {name=CommonMaterial, nodeColor=<rgba 0.5 0 1 1>}
  NamedRerouteDeclaration(None=@n50.GrassHeightmap) {name=GrassHeightmap, nodeColor=<rgba 0.5 0 1 1>}
  NamedRerouteDeclaration(None=@n50.GravelMaskForGrassFoliage) {name=GM-Gravel, nodeColor=<rgba 0.5 0 1 1>}
  NamedRerouteDeclaration(None=@n62.MaterialMask) {name=MF_ProceduralSnow-MatMask, nodeColor=<rgba 0.5 0 1 1>}
  NamedRerouteDeclaration(None=@n60.CliffMask) {name=CliffMask, nodeColor=<rgba 0.5 0 1 1>}
  NamedRerouteDeclaration(None=@n60.LandscapeGrassMeshMask) {name=GM-Cliff, nodeColor=<rgba 0.5 0 1 1>}
  n67 = GetMaterialAttributes(None=@n64.Result) {attributeGetTypes [
    #1
    #2
    #3
    #4
   ]}
  RuntimeVirtualTextureOutput!(
   BaseColor=MaterialFunctionCall!(BaseColor=@n67.Base Color) {materialFunction=/Game/Athena/Apollo/Environments/Landscape/Materials/Biome/CommonMaterials/MF_ReduceRTV-CompressionArtifacts}.BaseColor
   Specular=@n67.Specular
   Roughness=@n67.Specular
   Normal=@n67.Specular
   WorldHeight=ComponentMask(None=WorldPosition().XYZ) {b=true}
   Opacity
   Mask
   Displacement
   Mask4
  )
  SetMaterialAttributes!(MaterialAttributes, Base Color=Constant3Vector() {constant=<rgba 2 0 0 0>}) {attributeSetTypes [#1]}
 ]
}
